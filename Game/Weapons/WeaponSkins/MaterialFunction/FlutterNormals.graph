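# Weapon skin material function that layers an animated, panning normal map over the incoming material attributes' normal.
# A world-aligned triplanar call (WorldAlignedNormal_ComplexPan) projects the Normal Map, panned on all three planes by time and TriPlanner Speed.
# FlattenNormal fades it using a Fresnel-driven flatness, then BlendAngleCorrectedNormals merges it with the base normal.
# A pre-skinned local-space triplanar sampling chain is also present but unconnected to the output.
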MaterialFunction /Game/Weapons/WeaponSkins/MaterialFunction/FlutterNormals
graph {nodes [
  Divide()
  n1 = GetMaterialAttributes(None=FunctionInput() {inputType=FunctionInput_MaterialAttributes}) {attributeGetTypes [#1]}
  n2 = TextureObjectParameter() {
   group=PanningEmissive
   parameterName=Normal Map
   samplerType=SAMPLERTYPE_Normal
   texture=/Engine/Functions/Engine_MaterialFunctions02/ExampleContent/Textures/water_n
  }
  n3 = ScalarParameter() {defaultValue=500, group=PanningEmissive, parameterName=Normal Tiling}
  n4 = ScalarParameter() {defaultValue=1, group=PanningEmissive, parameterName=TriPlanner_AxisBlendStrength}
  n5 = Time()
  n6 = VectorParameter() {group=PanningEmissive, parameterName=TriPlanner Speed, defaultValue=<rgba 0 0.2 0.2 1>}
  n7 = Frac(None=Divide(A=Multiply(A=@n5, B=@n6.RGB)) {constB=5000})
  n8 = Constant() {r=1}
  n9 = VertexNormalWS()
  n10 = MaterialFunctionCall(
   Normal=MaterialFunctionCall!(
    TextureObject=@n2
    TextureSize=@n3
    Normal
    WorldSpace
    ProjectionTransitionContrast=@n4
    WorldPosition
    Use High Quality Normals
    Pan_YZ=@n7
    Pan_XZ=@n7
    Pan_XY=@n7
   ) {materialFunction=/Game/Weapons/WeaponSkins/Wraps/MaterialFunctions/WorldAlignedNormal_ComplexPan}.XYZ Texture
   Flatness=LinearInterpolate(A=ScalarParameter() {parameterName=Flatness}, Alpha=Fresnel(ExponentIn=@n8, Normal=@n9) {baseReflectFraction=0})
  ) {materialFunction=&FlattenNormal}
  FunctionOutput(None=SetMaterialAttributes!(
    MaterialAttributes=@n1.MaterialAttributes
    Normal=MaterialFunctionCall(BaseNormal=@n1.Normal, AdditionalNormal=@n10.Result) {materialFunction=&BlendAngleCorrectedNormals}.Result
   ) {attributeSetTypes [#1]})
  n11 = PreSkinnedPosition()
  n12 = Divide(
   A=FunctionInput(Preview=VertexInterpolator(VS=@n11).PS) {bUsePreviewValueAsDefault=true, inputName=Local Position}
   B=ScalarParameter() {defaultValue=40, group=PanningEmissive, parameterName=Triplanner_ConstantScaleDivisor}
  )
  n13 = Add(A=Multiply(A=@n3, B=@n12), B=@n7)
  n14 = TextureSample(UVs=Add(B=ComponentMask(None=@n13) {b=true, g=true}), Tex=@n2) {samplerType=SAMPLERTYPE_Masks, texture=/Game/Weapons/WeaponSkins/Textures/T_Swirly_Noise_highContrast}
  n15 = Add(A=Multiply(A=@n3, B=@n12), B=@n7)
  n16 = TextureSample(UVs=Add(B=ComponentMask(None=@n15) {b=true, r=true}), Tex=@n2) {samplerType=SAMPLERTYPE_Masks, texture=&T_Swirly_Noise_highContrast}
  n17 = FunctionInput(Preview=VertexInterpolator(VS=PreSkinnedNormal()).PS) {bUsePreviewValueAsDefault=true, inputName=Local Normal}
  n18 = Abs(None=ComponentMask(None=@n17) {g=true})
  n19 = Saturate(None=ComponentMask(None=MaterialFunctionCall(In=@n18, Contrast=@n4) {materialFunction=&CheapContrast}.Result) {g=true})
  n20 = Add(A=Multiply(A=@n3, B=@n12), B=@n7)
  n21 = TextureSample(UVs=Add(B=ComponentMask(None=@n20) {g=true, r=true}), Tex=@n2) {samplerType=SAMPLERTYPE_Masks, texture=&T_Swirly_Noise_highContrast}
  n22 = ComponentMask(None=@n17) {b=true}
  n23 = ComponentMask(None=MaterialFunctionCall(In=Abs(None=@n22), Contrast=@n4) {materialFunction=&CheapContrast}.Result) {b=true}
  LinearInterpolate(
   A=LinearInterpolate(A=LinearInterpolate(A=@n14.RGB, B=@n16.RGB, Alpha=@n19), B=@n21.RGB, Alpha=Saturate(None=@n23))
   B=LinearInterpolate(A=@n16.RGB, B=@n21.RGB, Alpha=FunctionInput() {bUsePreviewValueAsDefault=true, inputName=isVehicle, inputType=FunctionInput_Scalar})
   Alpha=ScalarParameter() {group=PanningEmissive, parameterName=Triplanner_Z axis Only}
  )
  MaterialFunctionCall() {materialFunction=&WorldAlignedNormal}
  MaterialFunctionCall() {materialFunction=&WorldAlignedNormal}
  VertexNormalWS()
 ]}
