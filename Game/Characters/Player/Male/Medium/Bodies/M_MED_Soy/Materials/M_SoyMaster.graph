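# Master skeletal-mesh material for the Soy player character body, built between the shared character PreFX and PostFX functions.
# FlutterNormals alters the PreFX result, and the red channel of a body mask blends back to the unfluttered attributes before a thin-film function.
# A parameterized Fresnel tinted by base color is added to emissive as a rim glow, and PreFX metallic gets a multiplier before PostFX.
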Material /Game/Characters/Player/Male/Medium/Bodies/M_MED_Soy/Materials/M_SoyMaster
graph {
 properties {
  bUseMaterialAttributes=true
  bUsedWithMeshDeformer=true
  bUsedWithMorphTargets=true
  bUsedWithSkeletalMesh=true
  pixelDepthOffsetMode=PDOM_Legacy
  usedShadingModels=Default Lit
 }
 nodes [
  Constant3Vector() {constant=<rgba 0.8 0.8 0.8 0>}
  n1 = MaterialFunctionCall() {materialFunction=&BaseFN_CharacterFunction_PreFX}
  n2 = MaterialFunctionCall!(Local Normal, Local Position, In=@n1.Result, isVehicle) {materialFunction=/Game/Weapons/WeaponSkins/MaterialFunction/FlutterNormals}
  n3 = TextureSampleParameter2D() {
   parameterName=Remove Flutter
   samplerType=SAMPLERTYPE_LinearColor
   texture=/Game/Characters/Player/Male/Medium/Bodies/M_MED_Soy/Textures/BodyMask
  }
  n4 = GetMaterialAttributes(None=MaterialFunctionCall!(Alpha=Constant() {r=1}, Material Attributes=BlendMaterialAttributes(A=@n2.Result, B=@n1.Result, Alpha=@n3.R)) {materialFunction=/Game/Weapons/WeaponSkins/MaterialFunction/ThinFilmFunction}.Result) {attributeGetTypes [#1, #2]}
  n5 = ScalarParameter() {defaultValue=6, group=Soy, parameterName=FresnelExponent, sortPriority=0}
  n6 = ScalarParameter() {group=Soy, parameterName=BaseReflectFraction, sortPriority=1}
  n7 = Saturate(None=Multiply(
    A=GetMaterialAttributes(None=@n1.Result) {attributeGetTypes [#2]}.Base Color
    B=Fresnel(ExponentIn=@n5, BaseReflectFractionIn=@n6) {exponent=0.3}
   ))
  n8 = Multiply(A=Multiply(A=@n4.Base Color, B=ScalarParameter() {defaultValue=1.25, group=Soy, parameterName=Fresnessl_Basecolor_Mult}), B=@n7)
  n9 = GetMaterialAttributes(None=@n1.Result) {attributeGetTypes [#3]}
  n10 = ScalarParameter() {defaultValue=1, group=Soy, parameterName=MetallicMult}
  MaterialFunctionCall(
   AO=@n1.AO
   Distance Blend=@n1.AO
   Fresnel=@n1.AO
   Material=SetMaterialAttributes!(MaterialAttributes=@n4.MaterialAttributes, Emissive Color=Add(A=@n4.Emissive Color, B=@n8), Metallic=Multiply(A=@n9.Metallic, B=@n10)) {attributeSetTypes [#1, #3]}
  ) {materialFunction=&BaseFN_CharacterFunction_PostFX}
  SetMaterialAttributes!(MaterialAttributes=@n1.Result, Base Color=@n7) {attributeSetTypes [#2]}
 ]
}
